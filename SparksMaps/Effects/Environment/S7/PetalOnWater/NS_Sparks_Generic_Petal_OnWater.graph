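# Niagara system of pink petals drifting flat on water, tinted by time of day.
# Two GPU emitters spawn texture-masked petals within a user radius up to a spawn cap, ignore lifetime and kill those leaving the radius.
# Forces and velocity are flattened to XY, with curl noise and drag providing drift.
# Petals_High renders meshes that avoid distance-field surfaces, react to player movement and spin with speed, while Petals_Low renders simpler sprites.
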NiagaraSystem /SparksMaps/Effects/Environment/S7/PetalOnWater/NS_Sparks_Generic_Petal_OnWater
dependencies {
 usedDataInterfaces [/Script/Niagara.NiagaraDataInterfaceTexture, &NiagaraDataInterfaceSpriteRendererInfo]
 usedDynamicInputs [&self:ScratchDynamicInput]
 usedModules [
  &SystemState
  &self:SystemSpawnScript.SystemScriptSource.SystemScriptGraph.NiagaraNodeAssignment_0.SetVariables_#1
  &self:SystemSpawnScript.SystemScriptSource.SystemScriptGraph.NiagaraNodeAssignment_1.SetVariables_#2
 ]
}
emitters {
 Petals_High {
  simTarget=GPUComputeSim
  emitterUpdateScript [
   &EmitterState(
    Enable Distance Culling=true
    Life Cycle Mode=::NewEnumerator1
    Loop Duration=1
    Module.bApplySpawnCountScaleByDistance=false
    Module.bUseMaxDistance=true
    Module.bUseMinDistance=true
    Scalability Mode=::NewEnumerator1
    Scale Spawn Count=true
    Spawn Count Scale By Distance=DataInterface() {propertyValues {Curve {keys [<at 0 0>,<at 0.2 1>,<at 0.75 1>,<at 1 0>]}}}
    UseLoopDelay=false
   )
   &SpawnRate(SpawnRate=$User.SpawnRate, Use Spawn Probability=false)
  ]
  particleSpawnScript [
   &InitializeParticle(
    Color=&RandomRangeLinearColor(Minimum=<rgba 1.2 0.2265 0.4311 1>, Maximum=<rgba 1.2 0.7323 0.563 1>)
    Lifetime Max=5
    Lifetime Min=2
    Lifetime Mode=::NewEnumerator1
    Mesh Scale Mode=::NewEnumerator3
    Mesh Uniform Scale=0.015
    Sprite Size=<30 20>
    Sprite Size Mode=::NewEnumerator4
    UsePositionOffset=false
    Write Material Random=true
   )
   &self:RainTextureMasking(
    BoxCenter:NiagaraPosition=<-31108 -39865 1500>
    BoxExtents:Vector3f=<8000 8000 1>
    Position:NiagaraPosition=$Particles.Position
    RejectionTexture=DataInterface() {propertyValues {
      Texture=/Game/Effects/Fort_Effects/Textures/Tile/T_Caustics_Chromatic
      TextureUserParameter {parameter {name=None, typeDefHandle {registeredTypeIndex=2}, varData []}}
     }}
   ) {enabled=false, module=RainTextureMasking}
   &ShapeLocation(
    Cylinder Height=0
    Cylinder Radius=$User.RadiusSpread
    MaskSpawnGroup=false
    Non Uniform Scale=<1 1 0>
    Override Seed=false
    Shape Primitive=::NewEnumerator1
    Surface Only=true
    Surface Only Band Thickness:NiagaraFloat=&RandomRangeFloat(Minimum=0.8, Maximum=1)
    Use Endcaps In Surface Only Mode=false
    UseOwnerScale=true
   )
   &InitialMeshOrientation(
    Mesh Axis To Orient=<0 1 0>
    Mesh Orientation Mode=::NewEnumerator2
    Rotation=&RandomRangeVector(Minimum=<0 0 0>, Maximum=<0 0 1>)
    Use Rotation Vector=true
   )
   /SetVariables(Particles.VelocityClamp:Vector3f=&Multiply_Vector(A=$Particles.Velocity, B=<1 1 0>)) {isSetParameters=true, module=SetVariables_#3}
   &self:RainTextureMasking_01(
    AlphaThreshold=0.55
    BoxCenter:NiagaraPosition=$Engine.Owner.Position
    BoxExtents:Vector3f=<3000 3000 1>
    Position:NiagaraPosition=$Particles.Position
    RejectionTexture=DataInterface() {propertyValues {
      Texture=&2ndInfectedMapLinear
      TextureUserParameter {parameter {name=None, typeDefHandle {registeredTypeIndex=2}, varData []}}
     }}
   ) {module=RainTextureMasking_01}
   &ShapeLocation(MaskSpawnGroup=false, Non Uniform Scale=<1 1 0>, Override Seed=false, Sphere Radius=800, UseOwnerScale=true) {module=ShapeLocation001}
   &KillParticles(Kill Particles=&SetBoolByIntComparison(A=$Engine.Emitter.TotalSpawnedParticles, B=$User.SpawnCap, Comparison Type=::NewEnumerator0))
   &ShapeLocation(MaskSpawnGroup=false, Non Uniform Scale=<1 1 0>, Override Seed=false, Sphere Radius=250, UseOwnerScale=true) {module=ShapeLocation002}
  ]
  particleUpdateScript [
   &ParticleState(Kill Particles When Lifetime Has Elapsed=false)
   &ScaleColor(
    Scale RGB=&FN_Select_Vector_from_Time_Of_Day(
     Morning=&MakeVectorFromLinearColorRGB(Color=<rgba 0.6308 0.491 0.4969 1>)
     Day=&MakeVectorFromLinearColorRGB(Color=<rgba 1 1 1 1>)
     Evening=&MakeVectorFromLinearColorRGB(Color=<rgba 0.6867 0.4735 0.4508 1>)
     Night=&MakeVectorFromLinearColorRGB(Color=<rgba 1.025 1.1405 1.2 1>)
    )
    ScaleA=true
    ScaleRGB=true
   )
   &PointForce(
    Force Strength=&RandomRangeFloat(Minimum=0.25, Maximum=0.5)
    Module.Use Falloff=false
    Module.UseLinearForce=false
   ) {module=PointForce001}
   &KillParticlesInVolume(Invert Volume=true, Sphere Radius:NiagaraFloat=$User.RadiusSpread)
   /Niagara/Modules/Math/AvoidDistanceFieldSurfaces_GPU(Avoid Nearest Surface=true, Avoid Oncoming Surfaces=true, Falloff Distance=50, Max Trace Distance=150, Obstactle Avoidance Strength=150)
   &PointForce(
    Force Falloff Distance:NiagaraFloat=&RandomRangeFloat(Minimum=100, Maximum=100)
    Force Origin=$System.PlayerPosition
    Force Strength=&RandomRangeFloat(Minimum=150, Maximum=160)
    Linear Force Vector:Vector3f=$System.PlayerVelocity
    Module.Use Falloff=true
    Module.UseLinearForce=true
   )
   &InterpolateOverTime(Rate Of Change=3, Target Vector Value:Vector3f=$Particles.Velocity, Variable Type=::NewEnumerator2)
   &MeshRotationForce(Rotation=&MakeVector(
     X=0
     Y=0
     Z=&Multiply_Float(
      A=&RandomRangeFloat(Minimum=-1, Maximum=1)
      B=&VectorLength(Vector=$Particles.InterpolateOverTime.InterpolateOverTime_V3_.Moving Average)
     )
    ))
   &CurlNoiseForce(
    Noise Frequency=7
    Noise Strength=&RandomRangeFloat(Minimum=20, Maximum=50)
    Pan Noise=true
    Pan Noise Field=&RandomRangeVector(Minimum=<-0.05 -0.05 0.1>, Maximum=<0.05 0.05 0.05>)
   )
   &Drag(
    Drag=&RandomRangeFloat(Minimum=2.5, Maximum=4)
    Rotational Drag=20
    Use Linear Drag=true
    Use Rotational Drag=true
   )
   &SolveForcesAndVelocity(
    Clamp Velocity=false
    Force:Vector3f=&Multiply_Vector(A=$Transient.PhysicsForce, B=<1 1 0>)
    Limit Acceleration=false
    Velocity:Vector3f=&Multiply_Vector(A=$Particles.Velocity, B=<1 1 0>)
   )
   &AlignSpriteToMeshOrientation(
    Mesh Orientation Relative Sprite Alignment Vector=<1 0 0>
    Mesh Orientation Relative Sprite Facing Vector:Vector3f=<0 0 1>
    SetAlignment=true
    SetFacing=true
   )
   &DynamicMaterialParameters(
    Index 0 Param 1=0.15
    Module.Param0WriteEnabled=true
    Module.Param1WriteEnabled=true
    Module.Param2WriteEnabled=true
    Module.Param3WriteEnabled=true
   )
   &SubUVAnimation(
    Direct Frame Index:NiagaraFloat=&RandomRangeFloat(Minimum=1, Maximum=6)
    "Sprite Renderer "=DataInterface() {propertyValues {SpriteRenderer=&self:Petals_High_8.NiagaraSpriteRendererProperties_0}}
    SubUV Animation Mode:ENiagara_SubUVLookupModeV2=::NewEnumerator5
   )
   &CameraOffset(Camera Offset Amount=0.5)
  ]
  rendererClasses [&NiagaraMeshRendererProperties]
  renderers [{rendererClass=&NiagaraMeshRendererProperties, rendererIndex=0}]
 }
 Petals_Low {
  simTarget=GPUComputeSim
  emitterUpdateScript [
   &EmitterState(
    Enable Distance Culling=true
    Life Cycle Mode=::NewEnumerator1
    Loop Duration=1
    Module.bApplySpawnCountScaleByDistance=false
    Module.bUseMaxDistance=true
    Module.bUseMinDistance=true
    Scalability Mode=::NewEnumerator1
    Scale Spawn Count=true
    Spawn Count Scale By Distance=DataInterface() {propertyValues {Curve {keys [<at 0 0>,<at 0.2 1>,<at 0.75 1>,<at 1 0>]}}}
    UseLoopDelay=false
   )
   &SpawnRate(SpawnRate=50, Use Spawn Probability=false)
  ]
  particleSpawnScript [
   &InitializeParticle(
    Color=&RandomRangeLinearColor(Minimum=<rgba 0.7011 0.3515 0.4397 1>, Maximum=<rgba 0.9219 0.53 0.6417 1>)
    Lifetime Max=5
    Lifetime Min=2
    Lifetime Mode=::NewEnumerator1
    Mesh Scale Mode=::NewEnumerator3
    Mesh Uniform Scale=0.025
    Sprite Size=<18 10>
    Sprite Size Mode=::NewEnumerator4
    UsePositionOffset=false
    Write Material Random=true
   )
   &ShapeLocation(
    Cylinder Height=0
    Cylinder Radius=$User.RadiusSpread
    MaskSpawnGroup=false
    Non Uniform Scale=<1 1 0>
    Override Seed=false
    Shape Primitive=::NewEnumerator1
    Surface Only=true
    Surface Only Band Thickness:NiagaraFloat=&RandomRangeFloat(Minimum=0.4, Maximum=0.5)
    Use Endcaps In Surface Only Mode=false
    UseOwnerScale=true
   )
   &ApplyInitialForces()
   &InitialMeshOrientation(
    Mesh Axis To Orient=<0 1 0>
    Mesh Orientation Mode=::NewEnumerator2
    Rotation=&RandomRangeVector(Minimum=<0 0 0>, Maximum=<0 0 1>)
    Use Rotation Vector=true
   )
   /SetVariables(Particles.VelocityClamp:Vector3f=&Multiply_Vector(A=$Particles.Velocity, B=<1 1 0>)) {isSetParameters=true, module=SetVariables_#3}
   &self:RainTextureMasking_01(
    AlphaThreshold=0.55
    BoxCenter:NiagaraPosition=$Engine.Owner.Position
    BoxExtents:Vector3f=<3000 3000 1>
    Position:NiagaraPosition=$Particles.Position
    RejectionTexture=DataInterface() {propertyValues {
      Texture=&2ndInfectedMapLinear
      TextureUserParameter {parameter {name=None, typeDefHandle {registeredTypeIndex=2}, varData []}}
     }}
   ) {module=RainTextureMasking_01}
   &KillParticles(Kill Particles=&SetBoolByIntComparison(A=$Engine.Emitter.TotalSpawnedParticles, B=$User.SpawnCap, Comparison Type=::NewEnumerator0))
  ]
  particleUpdateScript [
   &ParticleState(Kill Particles When Lifetime Has Elapsed=false)
   &ScaleColor(
    Scale RGB=&FN_Select_Vector_from_Time_Of_Day(
     Morning=&MakeVectorFromLinearColorRGB(Color=<rgba 0.6308 0.491 0.4969 1>)
     Day=&MakeVectorFromLinearColorRGB(Color=<rgba 1 1 1 1>)
     Evening=&MakeVectorFromLinearColorRGB(Color=<rgba 0.6867 0.4735 0.4508 1>)
     Night=&MakeVectorFromLinearColorRGB(Color=<rgba 1.025 1.1405 1.2 1>)
    )
    ScaleA=true
    ScaleRGB=true
   )
   &PointForce(
    Force Strength=&RandomRangeFloat(Minimum=0.25, Maximum=0.5)
    Module.Use Falloff=false
    Module.UseLinearForce=false
   ) {module=PointForce001}
   &KillParticlesInVolume(Invert Volume=true, Sphere Radius:NiagaraFloat=$User.RadiusSpread)
   &CurlNoiseForce(
    Noise Frequency=2
    Noise Strength=&RandomRangeFloat(Minimum=10, Maximum=20)
    Pan Noise=true
    Pan Noise Field=&RandomRangeVector(Minimum=<-0.05 -0.05 0.1>, Maximum=<0.05 0.05 0.05>)
   )
   &Drag(
    Drag=&RandomRangeFloat(Minimum=1, Maximum=3)
    Rotational Drag=20
    Use Linear Drag=true
    Use Rotational Drag=true
   )
   &SolveForcesAndVelocity(
    Clamp Velocity=false
    Force:Vector3f=&Multiply_Vector(A=$Transient.PhysicsForce, B=<1 1 0>)
    Limit Acceleration=false
    Velocity:Vector3f=&Multiply_Vector(A=$Particles.Velocity, B=<1 1 0>)
   )
   &AlignSpriteToMeshOrientation(
    Mesh Orientation Relative Sprite Alignment Vector=<1 0 0>
    Mesh Orientation Relative Sprite Facing Vector:Vector3f=<0 0 1>
    SetAlignment=true
    SetFacing=true
   )
   &DynamicMaterialParameters(
    Index 0 Param 1=0.2
    Module.Param0WriteEnabled=true
    Module.Param1WriteEnabled=true
    Module.Param2WriteEnabled=true
    Module.Param3WriteEnabled=true
   )
   &SubUVAnimation(
    Direct Frame Index:NiagaraFloat=&RandomRangeFloat(Minimum=1, Maximum=6)
    "Sprite Renderer "=DataInterface() {propertyValues {SpriteRenderer=&self:Petals_Low_7.NiagaraSpriteRendererProperties_0}}
    SubUV Animation Mode:ENiagara_SubUVLookupModeV2=::NewEnumerator5
   )
  ]
  rendererClasses [&NiagaraSpriteRendererProperties]
  renderers [{rendererClass=&NiagaraSpriteRendererProperties}]
 }
}
summary {
 systemName=NS_Sparks_Generic_Petal_OnWater
 emitters [
  {emitterName=Petals_High, simTarget=GPUComputeSim, rendererClasses [&NiagaraMeshRendererProperties]}
  {emitterName=Petals_Low, simTarget=GPUComputeSim, rendererClasses [&NiagaraSpriteRendererProperties]}
 ]
 userVariables [
  User.CurlCurve {
   type=NiagaraDataInterfaceCurve
   description=""
   defaultValue {
    struct=NiagaraExt_VariableValue_DataInterface
    value {
     dataInterface=&self:NiagaraDataInterfaceCurve_1
     dataInterfaceClass=&NiagaraDataInterfaceCurve
    }
   }
  }
  User.FallCurve {
   type=NiagaraDataInterfaceCurve
   description=""
   defaultValue {
    struct=NiagaraExt_VariableValue_DataInterface
    value {
     dataInterface=&self:NiagaraDataInterfaceCurve_0
     dataInterfaceClass=&NiagaraDataInterfaceCurve
    }
   }
  }
  User.RadiusSpread {type=NiagaraFloat, defaultValue=1500, description=""}
  User.SpawnCap {type=NiagaraInt32, defaultValue=600}
  User.SpawnRate {type=NiagaraFloat, defaultValue=1500, description=""}
 ]
}
